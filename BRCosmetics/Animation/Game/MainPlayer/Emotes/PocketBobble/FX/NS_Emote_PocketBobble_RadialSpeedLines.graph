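# Niagara system for the PocketBobble emote that bursts short-lived, multicolored radial speed lines around the player.
# Speed_Lines_Radial spawns stretching sprites on a sphere, trims them with kill volumes, and feeds layered outline and glow renderers plus time-of-day-scaled lights.
# FrontEnd_Ground_Glow adds flickering floor-facing glow sprites only in the front end.
# Front-end modifier functions retune rates, speeds, offsets and bounds between menu and gameplay, while random indices pick color curves.
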NiagaraSystem /BRCosmetics/Animation/Game/MainPlayer/Emotes/PocketBobble/FX/NS_Emote_PocketBobble_RadialSpeedLines
dependencies {
 usedDataInterfaces [
  &NiagaraDataInterfaceArrayColor
  &NiagaraDataInterfaceColorCurve
  &NiagaraDataInterfaceCurve
  &NiagaraDataInterfaceVector2DCurve
 ]
 usedDynamicInputs [&SelectLinearColorFromArray, &RandomRangeVector]
 usedModules [&SystemState]
}
emitters {
 FrontEnd_Ground_Glow {
  emitterUpdateScript [
   &EmitterState(Life Cycle Mode=::NewEnumerator1, Loop Duration=1, UseLoopDelay=false)
   &SpawnRate(SpawnRate=&FN_FrontEndFloatModifier(Gameplay=0, Front End=35), Use Spawn Probability=false)
  ]
  particleSpawnScript [
   /SetVariables(Particles.ColorIndex:NiagaraInt32=&RandomRangeInt(Minimum=0, Maximum=3)) {isSetParameters=true}
   &InitializeParticle(
    Color Mode=::NewEnumerator0
    Lifetime=0.05
    Position Offset=<0 0 -119>
    Sprite Size Mode=::NewEnumerator1
    Uniform Sprite Size Max=600
    Uniform Sprite Size Min=350
    UsePositionOffset=true
    Write Material Random=true
   )
   &SpriteFacingAndAlignment(Sprite Facing=<0 0 1>, WriteAlignment=false, WriteFacing=true)
  ]
  particleUpdateScript [
   &ParticleState()
   &Color(
    Color=&MergeHelper_ColorSwitch(
     0=&ColorFromCurve(ColorCurve=DataInterface() {propertyValues {
        AlphaCurve {keys [<at 0 1 RCIM_Cubic>,<at 1 0 RCIM_Cubic>]}
        BlueCurve {keys [<at 0 10 RCIM_Cubic>,<at 0.8 1 RCIM_Cubic>]}
        GreenCurve {keys [<at 0 1 RCIM_Cubic>,<at 0.8 0 RCIM_Cubic>]}
        RedCurve {keys [<at 0 10 RCIM_Cubic>,<at 0.8 0.0833 RCIM_Cubic>]}
       }})
     1=&ColorFromCurve(ColorCurve=DataInterface() {propertyValues {
        AlphaCurve {keys [<at 0 1 RCIM_Cubic>,<at 1 0 RCIM_Cubic>]}
        BlueCurve {keys [<at 0 5 RCIM_Cubic>,<at 0.9 1 RCIM_Cubic>]}
        GreenCurve {keys [<at 0 3.3333 RCIM_Cubic>,<at 0.9 0 RCIM_Cubic>]}
        RedCurve {keys [<at 0 0 RCIM_Cubic>,<at 0.9 0.0833 RCIM_Cubic>]}
       }})
     2=&ColorFromCurve(ColorCurve=DataInterface() {propertyValues {
        AlphaCurve {keys [<at 0 1 RCIM_Cubic>,<at 1 0 RCIM_Cubic>]}
        BlueCurve {keys [<at 0 5 RCIM_Cubic>,<at 0.9 1 RCIM_Cubic>]}
        GreenCurve {keys [<at 0 0.8333 RCIM_Cubic>,<at 0.9 0 RCIM_Cubic>]}
        RedCurve {keys [<at 0 0 RCIM_Cubic>,<at 0.9 0.3333 RCIM_Cubic>]}
       }})
     3=&ColorFromCurve(ColorCurve=DataInterface() {propertyValues {
        AlphaCurve {keys [<at 0 1 RCIM_Cubic>,<at 1 0 RCIM_Cubic>]}
        BlueCurve {keys [<at 0 5 RCIM_Cubic>,<at 0.9 1 RCIM_Cubic>]}
        GreenCurve {keys [<at 0 1.25 RCIM_Cubic>,<at 0.9 0 RCIM_Cubic>]}
        RedCurve {keys [<at 0 0.5 RCIM_Cubic>,<at 0.9 0.3333 RCIM_Cubic>]}
       }})
     4=&ColorFromCurve(ColorCurve=DataInterface() {propertyValues {
        AlphaCurve {keys [<at 0 1 RCIM_Cubic>,<at 1 0 RCIM_Cubic>]}
        BlueCurve {keys [<at 0 5 RCIM_Cubic>,<at 0.9 1 RCIM_Cubic>]}
        GreenCurve {keys [<at 0 0.4167 RCIM_Cubic>,<at 0.9 0 RCIM_Cubic>]}
        RedCurve {keys [<at 0 0 RCIM_Cubic>,<at 0.9 0.3333 RCIM_Cubic>]}
       }})
     5=&ColorFromCurve(ColorCurve=DataInterface() {propertyValues {
        AlphaCurve {keys [<at 0 1 RCIM_Cubic>,<at 1 0 RCIM_Cubic>]}
        BlueCurve {keys [<at 0 2.5 RCIM_Cubic>,<at 0.9 0.1667 RCIM_Cubic>]}
        GreenCurve {keys [<at 0 5 RCIM_Cubic>,<at 0.9 0.5 RCIM_Cubic>]}
        RedCurve {keys [<at 0 0 RCIM_Cubic>,<at 0.9 0 RCIM_Cubic>]}
       }})
     Index=$Particles.ColorIndex
     Num=::NewEnumerator1
    )
    Scale Alpha=&FloatFromCurve(
     FloatCurve=DataInterface() {propertyValues {Curve {keys [
         {tangentMode=RCTM_User, time=0, value=0}
         {arriveTangentWeight=0.3333, interpMode=RCIM_Cubic, leaveTangentWeight=0.3333, tangentMode=RCTM_User, time=1, value=1}
        ]}}}
     Scale Curve=0.17
    )
    Scale Color=&VectorFromFloat(Value=&FloatFromCurve(FloatCurve=DataInterface() {propertyValues {Curve {keys [
          {
           arriveTangent=-0.1238
           arriveTangentWeight=0.2855
           interpMode=RCIM_Cubic
           leaveTangent=-0.1238
           leaveTangentWeight=0.2855
           tangentMode=RCTM_User
           time=0
           value=0.5
          }
          {arriveTangent=-1.5105, tangentMode=RCTM_User, time=0.85, value=2}
         ]}}}))
   )
  ]
  rendererClasses [&NiagaraSpriteRendererProperties]
  renderers [{rendererClass=&NiagaraSpriteRendererProperties}]
 }
 Speed_Lines_Radial {
  emitterUpdateScript [
   &EmitterState(Life Cycle Mode=::NewEnumerator1, Loop Duration=0.95, UseLoopDelay=false)
   &SpawnRate(SpawnRate=&FN_FrontEndFloatModifier(Gameplay=1100, Front End=2500), Use Spawn Probability=false) {module=SpawnRate001}
   &SpawnRate(SpawnGroup=2, SpawnRate=50, Use Spawn Probability=false)
  ]
  particleSpawnScript [
   /SetVariables(
    Particles.ColorIndex:NiagaraInt32=&RandomRangeInt(Minimum=0, Maximum=3)
    Particles.VisibilityTag=&MakeCustomIntFromBool(
     Bool=&SetBoolByFloatComparison(A=&RandomRangeFloat(Minimum=0, Maximum=1), B=0.575)
     True INT=0
     False INT=1
    )
    Particles.VisibilityTagLight:NiagaraInt32=&SpawnGroup()
   ) {isSetParameters=true}
   &InitializeParticle(
    Color Mode=::NewEnumerator0
    Lifetime=&FN_FrontEndFloatModifier(
     Gameplay=&RandomRangeFloat(Minimum=0.05, Maximum=0.17)
     Front End=&RandomRangeFloat(Minimum=0.1, Maximum=0.25)
    )
    Position Mode=::NewEnumerator1
    Sprite Size Max=&MergeHelper_Vector2Switch(0=<20 65>, 1=<22 70>, 2=?, 3=?, 4=?, Index=$Particles.VisibilityTag, Num=::NewEnumerator0)
    Sprite Size Min=&MergeHelper_Vector2Switch(0=<8 35>, 1=<10 40>, 2=?, 3=?, 4=?, Index=$Particles.VisibilityTag, Num=::NewEnumerator0)
    Sprite Size Mode=::NewEnumerator2
    Sprite UV Mode=::NewEnumerator3
    UsePositionOffset=false
    Write Material Random=true
   )
   &ShapeLocation(
    MaskSpawnGroup=false
    Offset=&FN_FrontEndVector3Modifier(Gameplay=<0 0 0>, Front End=<0 0 -22>)
    Offset Coordinate Space=::Local
    Offset Mode=::NewEnumerator0
    Override Seed=false
    Rotation Coordinate Space=::Simulation
    Sphere Radius=95
    UseOwnerScale=false
   ) {module=ShapeLocation003}
   &AddVelocity(
    Module.Use Falloff Distance=false
    Origin Offset=&FN_FrontEndVector3Modifier(Gameplay=<0 0 0>, Front End=<0 0 -22>)
    Origin Offset Coordinate Space=::Simulation
    Override Seed=false
    Velocity Mode=::NewEnumerator1
    Velocity Speed=&FN_FrontEndFloatModifier(
     Gameplay=&RandomRangeFloat(Minimum=500, Maximum=1000)
     Front End=&RandomRangeFloat(Minimum=300, Maximum=600)
    )
   ) {module=AddVelocity001}
   &CameraOffset(Camera Offset Amount=&FN_FrontEndFloatModifier(Gameplay=-140, Front End=-180))
  ]
  particleUpdateScript [
   &ParticleState()
   &Drag(Drag=2, Use Linear Drag=true, Use Rotational Drag=false)
   &SolveForcesAndVelocity(Clamp Velocity=false, Limit Acceleration=false)
   &Color(
    Color=&MergeHelper_ColorSwitch(
     0=&ColorFromCurve(ColorCurve=DataInterface() {propertyValues {
        AlphaCurve {keys [<at 0 1 RCIM_Cubic>,<at 1 0 RCIM_Cubic>]}
        BlueCurve {keys [<at 0 10 RCIM_Cubic>,<at 0.8 1 RCIM_Cubic>]}
        GreenCurve {keys [<at 0 0 RCIM_Cubic>,<at 0.8 0 RCIM_Cubic>]}
        RedCurve {keys [<at 0 10 RCIM_Cubic>,<at 0.8 0.0833 RCIM_Cubic>]}
       }})
     1=&ColorFromCurve(ColorCurve=DataInterface() {propertyValues {
        AlphaCurve {keys [<at 0 1 RCIM_Cubic>,<at 1 0 RCIM_Cubic>]}
        BlueCurve {keys [<at 0 5 RCIM_Cubic>,<at 0.9 1 RCIM_Cubic>]}
        GreenCurve {keys [<at 0 3.3333 RCIM_Cubic>,<at 0.9 0 RCIM_Cubic>]}
        RedCurve {keys [<at 0 0 RCIM_Cubic>,<at 0.9 0.0833 RCIM_Cubic>]}
       }})
     2=&ColorFromCurve(ColorCurve=DataInterface() {propertyValues {
        AlphaCurve {keys [<at 0 1 RCIM_Cubic>,<at 1 0 RCIM_Cubic>]}
        BlueCurve {keys [<at 0 5 RCIM_Cubic>,<at 0.8 1 RCIM_Cubic>]}
        GreenCurve {keys [<at 0 0 RCIM_Cubic>,<at 0.8 0.0833 RCIM_Cubic>]}
        RedCurve {keys [<at 0 1.6667 RCIM_Cubic>,<at 0.8 0 RCIM_Cubic>]}
       }})
     3=&ColorFromCurve(ColorCurve=DataInterface() {propertyValues {
        AlphaCurve {keys [<at 0 1 RCIM_Cubic>,<at 1 0 RCIM_Cubic>]}
        BlueCurve {keys [<at 0 5 RCIM_Cubic>,<at 0.9 1 RCIM_Cubic>]}
        GreenCurve {keys [<at 0 0.8333 RCIM_Cubic>,<at 0.9 0 RCIM_Cubic>]}
        RedCurve {keys [<at 0 0 RCIM_Cubic>,<at 0.9 0.3333 RCIM_Cubic>]}
       }})
     4=&ColorFromCurve(ColorCurve=DataInterface() {propertyValues {
        AlphaCurve {keys [<at 0 1 RCIM_Cubic>,<at 1 0 RCIM_Cubic>]}
        BlueCurve {keys [<at 0 5 RCIM_Cubic>,<at 0.9 1 RCIM_Cubic>]}
        GreenCurve {keys [<at 0 0.4167 RCIM_Cubic>,<at 0.9 0 RCIM_Cubic>]}
        RedCurve {keys [<at 0 0 RCIM_Cubic>,<at 0.9 0.3333 RCIM_Cubic>]}
       }})
     5=&ColorFromCurve(ColorCurve=DataInterface() {propertyValues {
        AlphaCurve {keys [<at 0 1 RCIM_Cubic>,<at 1 0 RCIM_Cubic>]}
        BlueCurve {keys [<at 0 2.5 RCIM_Cubic>,<at 0.9 0.1667 RCIM_Cubic>]}
        GreenCurve {keys [<at 0 5 RCIM_Cubic>,<at 0.9 0.5 RCIM_Cubic>]}
        RedCurve {keys [<at 0 0 RCIM_Cubic>,<at 0.9 0 RCIM_Cubic>]}
       }})
     Index=$Particles.ColorIndex
     Num=::NewEnumerator2
    )
    Scale Alpha=&FloatFromCurve(
     FloatCurve=DataInterface() {propertyValues {Curve {keys [
         {tangentMode=RCTM_User, time=0, value=0}
         {arriveTangentWeight=0.3333, interpMode=RCIM_Cubic, leaveTangentWeight=0.3333, tangentMode=RCTM_User, time=1, value=1}
        ]}}}
     Scale Curve=&MakeCustomFloatFromBool(
      Bool=&SetBoolByFloatComparison(A=&RandomRangeFloat(Minimum=0, Maximum=1), B=0.3)
      True Float=&RandomRangeFloat(Minimum=12, Maximum=17)
      False Float=&RandomRangeFloat(Minimum=1, Maximum=7)
     )
    )
    Scale Color=&VectorFromFloat(Value=&FloatFromCurve(FloatCurve=DataInterface() {propertyValues {Curve {keys [
          {
           arriveTangent=-0.1238
           arriveTangentWeight=0.2855
           interpMode=RCIM_Cubic
           leaveTangent=-0.1238
           leaveTangentWeight=0.2855
           tangentMode=RCTM_User
           time=0
           value=0.5
          }
          {arriveTangent=-1.5105, tangentMode=RCTM_User, time=0.85, value=2}
         ]}}}))
   )
   &ScaleSpriteSize(
    Non-Uniform Curve Scale=&RandomRangeVector2D(
     Minimum=&Multiply_Vector2D_ByFloat(Float=&FN_FrontEndFloatModifier(Gameplay=1, Front End=0.8), Vector2D=<0.8 0.6>)
     Maximum=&Multiply_Vector2D_ByFloat(Float=&FN_FrontEndFloatModifier(Gameplay=1, Front End=0.8), Vector2D=<1.8 0.9>)
    )
    Non-Uniform Curve Sprite Scale=DataInterface() {propertyValues {
      XCurve {
       defaultValue=1
       keys [{arriveTangent=4.9398, interpMode=RCIM_Cubic, leaveTangent=4.9398, tangentMode=RCTM_User, time=0, value=0.1}, <at 1 3 RCIM_Cubic>]
      }
      YCurve {keys [
        {arriveTangent=4.1255, interpMode=RCIM_Cubic, leaveTangent=4.1255, tangentMode=RCTM_User, time=0, value=1}
        {arriveTangent=0.4613, tangentMode=RCTM_User, time=1, value=3}
       ]}
     }}
    Scale Sprite Size Mode=::NewEnumerator3
   ) {module=ScaleSpriteSize002}
   /SetVariables(
    Particles.Linear Color_Outline:LinearColor=&MakeLinearColorFromVectorAndFloat(
     "Vector (RGB)"=<0 0 0>
     "Float (Alpha)"=&FloatFromCurve(
      FloatCurve=DataInterface() {propertyValues {Curve {keys [
          {arriveTangentWeight=0.3333, interpMode=RCIM_Cubic, leaveTangentWeight=0.3333, tangentMode=RCTM_User, time=0, value=1}
          {tangentMode=RCTM_User, time=1, value=0}
         ]}}}
      Scale Curve=&Multiply_Float(
       A=&MakeFloatFromLinearColor(LinearColor=$Particles.Color, Channel:ENiagaraFloatFromLinearColorOptions=::NewEnumerator3)
       B=3
      )
     )
    )
    Particles.SpriteSize_Outline:Vector2f=&Multiply_Vector2D(A=$Particles.SpriteSize, B=<1 1>)
   ) {isSetParameters=true}
   /SetVariables(
    Particles.Linear Color_Glow:LinearColor=&MultiplyLinearColor(
     Linear Color=$Particles.Color
     Scale Factor=&MakeVector4(
      X=1
      Y=1
      Z=1
      W=&FloatFromCurve(
       FloatCurve=DataInterface() {propertyValues {Curve {keys [
           {arriveTangent=-0.2204, interpMode=RCIM_Cubic, leaveTangent=-0.2204, leaveTangentWeight=0.2731, tangentMode=RCTM_User, time=0, value=1.75}
           {arriveTangent=-4.7379, tangentMode=RCTM_User, time=0.8, value=0.25}
          ]}}}
       Scale Curve=2
      )
     )
    )
    Particles.SpriteSize_Glow:Vector2f=&Multiply_Vector2D(A=$Particles.SpriteSize, B=<1.5 1.4>)
   ) {isSetParameters=true}
   /SetVariables(
    Particles.Linear Color_Glowy:LinearColor=&MultiplyLinearColor(
     Linear Color=$Particles.Color
     Scale Factor=&MakeVector4(
      X=0.5
      Y=0.5
      Z=0.5
      W=&FloatFromCurve(
       FloatCurve=DataInterface() {propertyValues {Curve {keys [
           {arriveTangent=-0.2204, interpMode=RCIM_Cubic, leaveTangent=-0.2204, leaveTangentWeight=0.2731, tangentMode=RCTM_User, time=0, value=1.75}
           {arriveTangent=-4.7379, tangentMode=RCTM_User, time=0.8, value=0.25}
          ]}}}
       Scale Curve=&RandomRangeFloat(Minimum=0.035, Maximum=0.15)
      )
     )
    )
    Particles.SpriteSize_Glowy:Vector2f=&Multiply_Vector2D(
     A=$Particles.SpriteSize
     B=&Multiply_Vector2D_ByFloat(
      Float=&RandomRangeFloat(Minimum=1.3, Maximum=1.8)
      Vector2D=&RandomRangeVector2D(
       Minimum=&Vector2DFromCurve(
        Vector2Curve=DataInterface() {propertyValues {
          XCurve {keys [
            {arriveTangent=1.7978, interpMode=RCIM_Cubic, leaveTangent=1.7978, tangentMode=RCTM_User, time=0, value=0}
            {arriveTangent=0.2962, interpMode=RCIM_Cubic, leaveTangent=0.2962, tangentMode=RCTM_User, time=0.5, value=0.7}
           ]}
          YCurve {keys [<at 0 1>,<at 1 1>]}
         }}
        ScaleCurve=<0.9 1.1>
       )
       Maximum=&Vector2DFromCurve(
        Vector2Curve=DataInterface() {propertyValues {
          XCurve {keys [
            {arriveTangent=2.6448, interpMode=RCIM_Cubic, leaveTangent=2.6448, tangentMode=RCTM_User, time=0, value=0}
            {arriveTangent=0.8687, interpMode=RCIM_Cubic, leaveTangent=0.8687, tangentMode=RCTM_User, time=0.5, value=1.3}
           ]}
          YCurve {keys [<at 0 1>,<at 1 1>]}
         }}
        ScaleCurve=<1.3 1.1>
       )
      )
     )
    )
   ) {isSetParameters=true}
   &KillParticlesInVolume(Kill Shape=::NewEnumerator2, Origin Offset:Vector3f=<0 0 -70>)
   &KillParticlesInVolume(
    Invert Volume=true
    Origin Offset:Vector3f=&FN_FrontEndVector3Modifier(Gameplay=<0 0 0>, Front End=<0 0 -26>)
    Sphere Radius:NiagaraFloat=&FN_FrontEndFloatModifier(Gameplay=115, Front End=72.5)
   ) {module=KillParticlesInVolume001}
   &KillParticles(Kill Particles=HlslExpression() {hlslExpression="Particles.Color.a <= 0.05"})
   &FN_Light_Attributes(
    Color=&Multiply_LinearColorByFloat(Linear Color=$Particles.Color, Float=1)
    Create Color Attribute=true
    Create Exponent Attribute=true
    Create Position Attribute=true
    Create Radius Attribute=true
    Exponent=2
    Radius=250
    Scale Color=&FN_Select_Float_from_Time_Of_Day(Morning=0.4, Day=0.35, Evening=0.4, Night=0.4)
   )
  ]
  rendererClasses [&NiagaraSpriteRendererProperties, &NiagaraLightRendererProperties]
  renderers [
   {rendererClass=&NiagaraSpriteRendererProperties}
   {rendererClass=&NiagaraSpriteRendererProperties, rendererIndex=1}
   {rendererClass=&NiagaraSpriteRendererProperties, rendererIndex=2}
   {rendererClass=&NiagaraSpriteRendererProperties, rendererIndex=3}
   {rendererClass=&NiagaraLightRendererProperties, rendererIndex=4}
  ]
 }
}
summary {
 systemName=NS_Emote_PocketBobble_RadialSpeedLines
 emitters [
  {
   emitterName=Speed_Lines_Radial
   rendererClasses [&NiagaraSpriteRendererProperties, &NiagaraLightRendererProperties]
  }
  {emitterName=FrontEnd_Ground_Glow, rendererClasses [&NiagaraSpriteRendererProperties]}
 ]
 userVariables [
  User.bIsFrontEnd {type=NiagaraBool}
  User.bIsFrontEndPreview {type=NiagaraBool}
  User.ColorVariant {type=NiagaraInt32}
  User.DelayTime {type=NiagaraFloat, defaultValue=0, description=""}
  User.FlashIntensity {type=NiagaraFloat, defaultValue=0.7, description=""}
  User.IsLocalPlayer {type=NiagaraBool, defaultValue=true}
  User.SplashEroder {type=NiagaraFloat, defaultValue=0.7, description=""}
  User.WorkoutSelect {type=NiagaraFloat, defaultValue=0, description=""}
 ]
}
