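# Niagara smear effect for the PocketBobble emote, drawing faint, flickering multicolor speed-line sprites.
# Two near-identical emitters each burst five locally aligned sprites, using the partition index to place them in a symmetric V, with the second shorter and fainter.
# Sprites stretch out quickly along X, pulse in length via a waveform, and cycle through bright color curves while fading out.
# None of the declared User parameters are referenced in the listed modules.
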NiagaraSystem /BRCosmetics/Animation/Game/MainPlayer/Emotes/PocketBobble/FX/NS_Emote_PocketBobble_Smear
dependencies {
 usedDataInterfaces [&NiagaraDataInterfaceCurve, &NiagaraDataInterfaceVector2DCurve, &NiagaraDataInterfaceColorCurve]
 usedDynamicInputs []
 usedModules [&SystemState]
}
emitters {
 Speed_Lines_Straight {
  emitterUpdateScript [
   &EmitterState(Life Cycle Mode=::NewEnumerator1, Loop Behavior=::NewEnumerator1, Loop Duration=1, UseLoopDelay=false)
   &SpawnBurst_Instantaneous(Spawn Count=5, Use Spawn Probability=false, UseLoopCountLimit=false)
  ]
  particleSpawnScript [
   &PartitionParticles()
   &InitializeParticle(
    Lifetime=2
    Position Offset=&MergeHelper_VectorSwitch(0=<0 0 0>, 1=<7.5 2.5 5>, 2=<-7.5 2.5 5>, 3=<15 5 15>, 4=<-15 5 15>, Index=$Particles.PartitionParticles.Partition, Num=::NewEnumerator3)
    Sprite Rotation Mode=::NewEnumerator0
    Sprite Size Max=<300 100>
    Sprite Size Min=<250 80>
    Sprite Size Mode=::NewEnumerator2
    UsePositionOffset=true
    Write Material Random=true
   )
   &SpriteFacingAndAlignment(Facing Coordinate Space=::Local, Sprite Alignment=<0 0 0.25>, WriteAlignment=true, WriteFacing=true)
  ]
  particleUpdateScript [
   &ParticleState()
   &ScaleSpriteSize(
    Non-Uniform Curve Scale=<1.2 1>
    Non-Uniform Curve Sprite Scale=DataInterface() {propertyValues {
      XCurve {keys [
        {arriveTangent=3.8849, interpMode=RCIM_Cubic, leaveTangent=3.8849, tangentMode=RCTM_User, time=0, value=0.25}
        {arriveTangent=1.49, interpMode=RCIM_Cubic, leaveTangent=1.49, tangentMode=RCTM_User, time=0.1871, value=0.9}
        <at 0.3829 1 RCIM_Cubic>
       ]}
      YCurve {keys [{tangentMode=RCTM_User, time=0, value=1}, {tangentMode=RCTM_User, time=1, value=1}]}
     }}
    Scale Sprite Size Mode=::NewEnumerator3
   )
   &ScaleSpriteSize(
    Scale Factor=&MakeVector2D(X=&Waveform("Waveform [1]"=::NewEnumerator12, "[1] Frequency"=1, "[1] Amplitude Min/Max"=<0.7 1.2>), Y=1)
    Scale Sprite Size Mode=::NewEnumerator2
   ) {module=ScaleSpriteSize001}
   &Color(
    Color=&ColorFromCurve(ColorCurve=DataInterface() {propertyValues {
       AlphaCurve {keys [<at 0 1 RCIM_Cubic>,<at 1 0 RCIM_Cubic>]}
       BlueCurve {keys [
         {arriveTangent=-17.6373, interpMode=RCIM_Cubic, time=0, value=5}
         {arriveTangent=-22.8791, interpMode=RCIM_Cubic, leaveTangent=-22.8791, time=0.0821, value=5}
         <at 0.153 1.5>
         <at 0.2371 3>
         <at 0.3316 2>
         <at 0.4172 1>
         <at 0.5161 1>
         <at 0.6136 1.5>
         <at 0.7391 1>
         <at 0.8262 1>
         <at 0.9221 1>
         <at 0.993 1>
        ]}
       GreenCurve {keys [
         <at 0 2.0523 RCIM_Cubic>
         {arriveTangent=-8.5128, interpMode=RCIM_Cubic, leaveTangent=-8.5128, time=0.0821, value=0.25}
         <at 0.153 0.75>
         <at 0.2371 0.15>
         <at 0.3316 1>
         <at 0.4172 0>
         <at 0.5161 0.5>
         <at 0.6136 0>
         <at 0.7391 0.5>
         <at 0.8262 0>
         <at 0.9221 0.0823>
         <at 0.993 0>
        ]}
       RedCurve {keys [
         {arriveTangent=-10.2885, interpMode=RCIM_Cubic, time=0, value=0}
         <at 0.0821 3.4167 RCIM_Cubic>
         <at 0.153 0>
         <at 0.2371 2.525>
         <at 0.3316 0>
         <at 0.4172 1>
         <at 0.5161 0>
         <at 0.6136 0.5426>
         <at 0.7391 0>
         <at 0.8262 0.4167>
         <at 0.9221 0>
         <at 0.993 0.6539>
        ]}
      }})
    Scale Alpha=&RandomRangeFloat(Minimum=0.07, Maximum=0.14)
   ) {module=Color001}
  ]
  rendererClasses [&NiagaraSpriteRendererProperties]
  renderers [{rendererClass=&NiagaraSpriteRendererProperties}]
 }
 Speed_Lines_Straight001 {
  emitterUpdateScript [
   &EmitterState(Life Cycle Mode=::NewEnumerator1, Loop Behavior=::NewEnumerator1, Loop Duration=1, UseLoopDelay=false)
   &SpawnBurst_Instantaneous(Spawn Count=5, Use Spawn Probability=false, UseLoopCountLimit=false)
  ]
  particleSpawnScript [
   &PartitionParticles()
   &InitializeParticle(
    Lifetime=2
    Position Offset=&MergeHelper_VectorSwitch(0=<0 0 0>, 1=<7.5 2.5 5>, 2=<-7.5 2.5 5>, 3=<15 5 15>, 4=<-15 5 15>, Index=$Particles.PartitionParticles.Partition, Num=::NewEnumerator3)
    Sprite Rotation Mode=::NewEnumerator0
    Sprite Size Max=<200 100>
    Sprite Size Min=<175 80>
    Sprite Size Mode=::NewEnumerator2
    UsePositionOffset=true
    Write Material Random=true
   )
   &SpriteFacingAndAlignment(Facing Coordinate Space=::Local, Sprite Alignment=<0 0 0.25>, WriteAlignment=true, WriteFacing=true)
  ]
  particleUpdateScript [
   &ParticleState()
   &ScaleSpriteSize(
    Non-Uniform Curve Scale=<1.1 1>
    Non-Uniform Curve Sprite Scale=DataInterface() {propertyValues {
      XCurve {keys [
        {arriveTangent=6.1956, interpMode=RCIM_Cubic, leaveTangent=6.1956, tangentMode=RCTM_User, time=0, value=0.25}
        {arriveTangent=1.49, interpMode=RCIM_Cubic, leaveTangent=1.49, tangentMode=RCTM_User, time=0.1231, value=0.9}
        <at 0.319 1 RCIM_Cubic>
       ]}
      YCurve {keys [{tangentMode=RCTM_User, time=0, value=1}, {tangentMode=RCTM_User, time=1, value=1}]}
     }}
    Scale Sprite Size Mode=::NewEnumerator3
   )
   &ScaleSpriteSize(
    Scale Factor=&MakeVector2D(X=&Waveform("Waveform [1]"=::NewEnumerator12, "[1] Frequency"=1, "[1] Amplitude Min/Max"=<0.7 1.2>), Y=1)
    Scale Sprite Size Mode=::NewEnumerator2
   ) {module=ScaleSpriteSize001}
   &Color(
    Color=&ColorFromCurve(ColorCurve=DataInterface() {propertyValues {
       AlphaCurve {keys [<at 0 1 RCIM_Cubic>,<at 1 0 RCIM_Cubic>]}
       BlueCurve {keys [
         {arriveTangent=-17.6373, interpMode=RCIM_Cubic, time=0, value=5}
         {arriveTangent=-22.8791, interpMode=RCIM_Cubic, leaveTangent=-22.8791, time=0.0821, value=5}
         <at 0.153 1.5>
         <at 0.2371 3>
         <at 0.3316 2>
         <at 0.4172 1>
         <at 0.5161 1>
         <at 0.6136 1.5>
         <at 0.7391 1>
         <at 0.8262 1>
         <at 0.9221 1>
         <at 0.993 1>
        ]}
       GreenCurve {keys [
         <at 0 2.0523 RCIM_Cubic>
         {arriveTangent=-8.5128, interpMode=RCIM_Cubic, leaveTangent=-8.5128, time=0.0821, value=0.25}
         <at 0.153 0.75>
         <at 0.2371 0.15>
         <at 0.3316 1>
         <at 0.4172 0>
         <at 0.5161 0.5>
         <at 0.6136 0>
         <at 0.7391 0.5>
         <at 0.8262 0>
         <at 0.9221 0.0823>
         <at 0.993 0>
        ]}
       RedCurve {keys [
         {arriveTangent=-10.2885, interpMode=RCIM_Cubic, time=0, value=0}
         <at 0.0821 3.4167 RCIM_Cubic>
         <at 0.153 0>
         <at 0.2371 2.525>
         <at 0.3316 0>
         <at 0.4172 1>
         <at 0.5161 0>
         <at 0.6136 0.5426>
         <at 0.7391 0>
         <at 0.8262 0.4167>
         <at 0.9221 0>
         <at 0.993 0.6539>
        ]}
      }})
    Scale Alpha=&RandomRangeFloat(Minimum=0.05, Maximum=0.1)
   ) {module=Color001}
  ]
  rendererClasses [&NiagaraSpriteRendererProperties]
  renderers [{rendererClass=&NiagaraSpriteRendererProperties}]
 }
}
summary {
 systemName=NS_Emote_PocketBobble_Smear
 emitters [
  {emitterName=Speed_Lines_Straight, rendererClasses [&NiagaraSpriteRendererProperties]}
  {emitterName=Speed_Lines_Straight001, rendererClasses [&NiagaraSpriteRendererProperties]}
 ]
 userVariables [
  User.bIsFrontEnd {type=NiagaraBool}
  User.bIsFrontEndPreview {type=NiagaraBool}
  User.ColorVariant {type=NiagaraInt32}
  User.DelayTime {type=NiagaraFloat, defaultValue=0, description=""}
  User.FlashIntensity {type=NiagaraFloat, defaultValue=0.7, description=""}
  User.IsLocalPlayer {type=NiagaraBool, defaultValue=true}
  User.RibbonIDOffset {type=NiagaraInt32}
  User.SplashEroder {type=NiagaraFloat, defaultValue=0.7, description=""}
  User.Trail_Life {type=NiagaraFloat, defaultValue=0, description=""}
  User.WorkoutSelect {type=NiagaraFloat, defaultValue=0, description=""}
 ]
}
