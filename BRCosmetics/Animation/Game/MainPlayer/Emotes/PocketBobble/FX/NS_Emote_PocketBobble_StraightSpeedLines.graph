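# Niagara effect for the PocketBobble emote that emits short-lived, neon-colored straight speed lines driven along negative Y and slowed by drag.
# Lines pick random colors and two size classes, then derive outline and glow attributes for multiple sprite renderers, plus a light renderer scaled by time of day.
# A second emitter adds a flickering ground glow only in the front end, where line spawn rate also rises and a preview flag culls particles outside a box.
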NiagaraSystem /BRCosmetics/Animation/Game/MainPlayer/Emotes/PocketBobble/FX/NS_Emote_PocketBobble_StraightSpeedLines
dependencies {
 usedDataInterfaces [&NiagaraDataInterfaceArrayColor, &NiagaraDataInterfaceColorCurve, &NiagaraDataInterfaceCurve]
 usedDynamicInputs [&MergeHelper_FloatSwitch]
 usedModules [&SystemState]
}
emitters {
 FrontEnd_Ground_Glow {
  emitterUpdateScript [
   &EmitterState(Life Cycle Mode=::NewEnumerator1, Loop Duration=1, UseLoopDelay=false)
   &SpawnRate(SpawnRate=&FN_FrontEndFloatModifier(Gameplay=0, Front End=35), Use Spawn Probability=false)
  ]
  particleSpawnScript [
   /SetVariables(Particles.ColorIndex:NiagaraInt32=&RandomRangeInt(Minimum=0, Maximum=3)) {isSetParameters=true}
   &InitializeParticle(
    Color Mode=::NewEnumerator0
    Lifetime=0.06
    Position Offset=<0 -25 -79.5>
    Sprite Size Mode=::NewEnumerator1
    Uniform Sprite Size Max=650
    Uniform Sprite Size Min=400
    UsePositionOffset=true
    Write Material Random=true
   )
   &SpriteFacingAndAlignment(Sprite Facing=<0 0 1>, WriteAlignment=false, WriteFacing=true)
  ]
  particleUpdateScript [
   &ParticleState()
   &Color(
    Color=&MergeHelper_ColorSwitch(
     0=&ColorFromCurve(ColorCurve=DataInterface() {propertyValues {
        AlphaCurve {keys [<at 0 1 RCIM_Cubic>,<at 1 0 RCIM_Cubic>]}
        BlueCurve {keys [<at 0 10 RCIM_Cubic>,<at 0.8 1 RCIM_Cubic>]}
        GreenCurve {keys [<at 0 1 RCIM_Cubic>,<at 0.8 0 RCIM_Cubic>]}
        RedCurve {keys [<at 0 10 RCIM_Cubic>,<at 0.8 0.0833 RCIM_Cubic>]}
       }})
     1=&ColorFromCurve(ColorCurve=DataInterface() {propertyValues {
        AlphaCurve {keys [<at 0 1 RCIM_Cubic>,<at 1 0 RCIM_Cubic>]}
        BlueCurve {keys [<at 0 5 RCIM_Cubic>,<at 0.9 1 RCIM_Cubic>]}
        GreenCurve {keys [<at 0 3.3333 RCIM_Cubic>,<at 0.9 0 RCIM_Cubic>]}
        RedCurve {keys [<at 0 0 RCIM_Cubic>,<at 0.9 0.0833 RCIM_Cubic>]}
       }})
     2=&ColorFromCurve(ColorCurve=DataInterface() {propertyValues {
        AlphaCurve {keys [<at 0 1 RCIM_Cubic>,<at 1 0 RCIM_Cubic>]}
        BlueCurve {keys [<at 0 5 RCIM_Cubic>,<at 0.9 1 RCIM_Cubic>]}
        GreenCurve {keys [<at 0 0.8333 RCIM_Cubic>,<at 0.9 0 RCIM_Cubic>]}
        RedCurve {keys [<at 0 0 RCIM_Cubic>,<at 0.9 0.3333 RCIM_Cubic>]}
       }})
     3=&ColorFromCurve(ColorCurve=DataInterface() {propertyValues {
        AlphaCurve {keys [<at 0 1 RCIM_Cubic>,<at 1 0 RCIM_Cubic>]}
        BlueCurve {keys [<at 0 5 RCIM_Cubic>,<at 0.9 1 RCIM_Cubic>]}
        GreenCurve {keys [<at 0 1.25 RCIM_Cubic>,<at 0.9 0 RCIM_Cubic>]}
        RedCurve {keys [<at 0 0.5 RCIM_Cubic>,<at 0.9 0.3333 RCIM_Cubic>]}
       }})
     4=&ColorFromCurve(ColorCurve=DataInterface() {propertyValues {
        AlphaCurve {keys [<at 0 1 RCIM_Cubic>,<at 1 0 RCIM_Cubic>]}
        BlueCurve {keys [<at 0 5 RCIM_Cubic>,<at 0.9 1 RCIM_Cubic>]}
        GreenCurve {keys [<at 0 0.4167 RCIM_Cubic>,<at 0.9 0 RCIM_Cubic>]}
        RedCurve {keys [<at 0 0 RCIM_Cubic>,<at 0.9 0.3333 RCIM_Cubic>]}
       }})
     5=&ColorFromCurve(ColorCurve=DataInterface() {propertyValues {
        AlphaCurve {keys [<at 0 1 RCIM_Cubic>,<at 1 0 RCIM_Cubic>]}
        BlueCurve {keys [<at 0 2.5 RCIM_Cubic>,<at 0.9 0.1667 RCIM_Cubic>]}
        GreenCurve {keys [<at 0 5 RCIM_Cubic>,<at 0.9 0.5 RCIM_Cubic>]}
        RedCurve {keys [<at 0 0 RCIM_Cubic>,<at 0.9 0 RCIM_Cubic>]}
       }})
     Index=$Particles.ColorIndex
     Num=::NewEnumerator1
    )
    Scale Alpha=&FloatFromCurve(
     FloatCurve=DataInterface() {propertyValues {Curve {keys [
         {tangentMode=RCTM_User, time=0, value=0}
         {arriveTangentWeight=0.3333, interpMode=RCIM_Cubic, leaveTangentWeight=0.3333, tangentMode=RCTM_User, time=1, value=1}
        ]}}}
     Scale Curve=0.2
    )
    Scale Color=&VectorFromFloat(Value=&FloatFromCurve(FloatCurve=DataInterface() {propertyValues {Curve {keys [
          {
           arriveTangent=-0.1238
           arriveTangentWeight=0.2855
           interpMode=RCIM_Cubic
           leaveTangent=-0.1238
           leaveTangentWeight=0.2855
           tangentMode=RCTM_User
           time=0
           value=0.5
          }
          {arriveTangent=-1.5105, tangentMode=RCTM_User, time=0.85, value=2}
         ]}}}))
   )
  ]
  rendererClasses [&NiagaraSpriteRendererProperties]
  renderers [{rendererClass=&NiagaraSpriteRendererProperties}]
 }
 Speed_Lines_Straight001 {
  emitterUpdateScript [
   &EmitterState(Life Cycle Mode=::NewEnumerator1, Loop Duration=0.95, UseLoopDelay=false)
   &SpawnRate(SpawnRate=&FN_FrontEndFloatModifier(Gameplay=175, Front End=250), Use Spawn Probability=false) {module=SpawnRate001}
   &SpawnRate(SpawnGroup=2, SpawnRate=50, Use Spawn Probability=false)
  ]
  particleSpawnScript [
   /SetVariables(
    Particles.ColorIndex:NiagaraInt32=&RandomRangeInt(Minimum=0, Maximum=4)
    Particles.VisibilityTag=&MakeCustomIntFromBool(
     Bool=&SetBoolByFloatComparison(A=&RandomRangeFloat(Minimum=0, Maximum=1), B=0.4)
     True INT=0
     False INT=1
    )
   ) {isSetParameters=true}
   &InitializeParticle(
    Color=&SelectLinearColorFromArray(
     Array Sampling Mode=::NewEnumerator1
     Direct Array Index=&RandomRangeInt(Minimum=0, Maximum=4)
     Color Selection Array=DataInterface() {propertyValues {ColorData [<rgba 0.04 0.04 0.04 1>, <rgba 0 1 0.4167 1>, <rgba 2 0.1565 2 1>, <rgba 2 0.3652 0.6377 1>, <rgba 0 0.3333 2 1>]}}
     Random Seed=?
    )
    Lifetime Max=0.14
    Lifetime Min=0.095
    Lifetime Mode=::NewEnumerator1
    Position Mode=::NewEnumerator1
    Sprite Size Max=&MergeHelper_Vector2Switch(0=<50 125>, 1=<40 200>, 2=?, 3=?, 4=?, Index=$Particles.VisibilityTag, Num=::NewEnumerator0)
    Sprite Size Min=&MergeHelper_Vector2Switch(0=<3 41>, 1=<5 80>, 2=?, 3=?, 4=?, Index=$Particles.VisibilityTag, Num=::NewEnumerator0)
    Sprite Size Mode=::NewEnumerator2
    Sprite UV Mode=::NewEnumerator3
    UsePositionOffset=false
    Write Material Random=true
   )
   &AddVelocity(Velocity=&RandomRangeVector(Minimum=<0 -500 0>, Maximum=<1 -50 0>)) {module=AddVelocity001}
   &CameraOffset(Camera Offset Amount=&FN_FrontEndFloatModifier(Gameplay=0, Front End=-30))
   &ShapeLocation(Box Size=<75 75 100>, MaskSpawnGroup=false, Override Seed=false, Shape Primitive=::NewEnumerator2, Surface Only=false, UseOwnerScale=true)
   /SetVariables(Particles.Color=&MakeCustomLinearColorFromBool(
     Bool=&SetBoolByIntComparison(A=$Particles.ColorIndex, B=0, Comparison Type=::NewEnumerator4)
     True Color=&MultiplyLinearColor(
      Linear Color=$Particles.Color
      Scale Factor=&MakeVector4FromVectorAndFloat(
       "Vector (RGB)"=&VectorFromFloat(Value=&RandomRangeFloat(Minimum=0.5, Maximum=3))
       "Float (Alpha)"=&RandomRangeFloat(Minimum=0.5, Maximum=3)
      )
     )
     False Color=$Particles.Color
    )) {isSetParameters=true}
  ]
  particleUpdateScript [
   &ParticleState()
   &Drag(Drag=5, Use Linear Drag=true, Use Rotational Drag=false)
   &SolveForcesAndVelocity(Clamp Velocity=false, Limit Acceleration=false)
   &Color(
    Color=&MergeHelper_ColorSwitch(
     0=&ColorFromCurve(ColorCurve=DataInterface() {propertyValues {
        AlphaCurve {keys [<at 0 1 RCIM_Cubic>,<at 1 0 RCIM_Cubic>]}
        BlueCurve {keys [<at 0 10 RCIM_Cubic>,<at 0.8 1 RCIM_Cubic>]}
        GreenCurve {keys [<at 0 0 RCIM_Cubic>,<at 0.8 0 RCIM_Cubic>]}
        RedCurve {keys [<at 0 10 RCIM_Cubic>,<at 0.8 0.0833 RCIM_Cubic>]}
       }})
     1=&ColorFromCurve(ColorCurve=DataInterface() {propertyValues {
        AlphaCurve {keys [<at 0 1 RCIM_Cubic>,<at 1 0 RCIM_Cubic>]}
        BlueCurve {keys [<at 0 5 RCIM_Cubic>,<at 0.9 1 RCIM_Cubic>]}
        GreenCurve {keys [<at 0 3.3333 RCIM_Cubic>,<at 0.9 0 RCIM_Cubic>]}
        RedCurve {keys [<at 0 0 RCIM_Cubic>,<at 0.9 0.0833 RCIM_Cubic>]}
       }})
     2=&ColorFromCurve(ColorCurve=DataInterface() {propertyValues {
        AlphaCurve {keys [<at 0 1 RCIM_Cubic>,<at 1 0 RCIM_Cubic>]}
        BlueCurve {keys [<at 0 5 RCIM_Cubic>,<at 0.8 1 RCIM_Cubic>]}
        GreenCurve {keys [<at 0 0 RCIM_Cubic>,<at 0.8 0.0833 RCIM_Cubic>]}
        RedCurve {keys [<at 0 1.6667 RCIM_Cubic>,<at 0.8 0 RCIM_Cubic>]}
       }})
     3=&ColorFromCurve(ColorCurve=DataInterface() {propertyValues {
        AlphaCurve {keys [<at 0 1 RCIM_Cubic>,<at 1 0 RCIM_Cubic>]}
        BlueCurve {keys [<at 0 5 RCIM_Cubic>,<at 0.9 1 RCIM_Cubic>]}
        GreenCurve {keys [<at 0 0.8333 RCIM_Cubic>,<at 0.9 0 RCIM_Cubic>]}
        RedCurve {keys [<at 0 0 RCIM_Cubic>,<at 0.9 0.3333 RCIM_Cubic>]}
       }})
     4=&ColorFromCurve(ColorCurve=DataInterface() {propertyValues {
        AlphaCurve {keys [<at 0 1 RCIM_Cubic>,<at 1 0 RCIM_Cubic>]}
        BlueCurve {keys [<at 0 5 RCIM_Cubic>,<at 0.9 1 RCIM_Cubic>]}
        GreenCurve {keys [<at 0 0.4167 RCIM_Cubic>,<at 0.9 0 RCIM_Cubic>]}
        RedCurve {keys [<at 0 0 RCIM_Cubic>,<at 0.9 0.3333 RCIM_Cubic>]}
       }})
     5=&ColorFromCurve(ColorCurve=DataInterface() {propertyValues {
        AlphaCurve {keys [<at 0 1 RCIM_Cubic>,<at 1 0 RCIM_Cubic>]}
        BlueCurve {keys [<at 0 2.5 RCIM_Cubic>,<at 0.9 0.1667 RCIM_Cubic>]}
        GreenCurve {keys [<at 0 5 RCIM_Cubic>,<at 0.9 0.5 RCIM_Cubic>]}
        RedCurve {keys [<at 0 0 RCIM_Cubic>,<at 0.9 0 RCIM_Cubic>]}
       }})
     Index=$Particles.ColorIndex
     Num=::NewEnumerator2
    )
    Scale Alpha=&FloatFromCurve(
     FloatCurve=DataInterface() {propertyValues {Curve {keys [
         {tangentMode=RCTM_User, time=0, value=0}
         {arriveTangentWeight=0.3333, interpMode=RCIM_Cubic, leaveTangentWeight=0.3333, tangentMode=RCTM_User, time=1, value=1}
        ]}}}
     Scale Curve=20
    )
    Scale Color=&VectorFromFloat(Value=&FloatFromCurve(FloatCurve=DataInterface() {propertyValues {Curve {keys [
          {
           arriveTangent=-0.1238
           arriveTangentWeight=0.2855
           interpMode=RCIM_Cubic
           leaveTangent=-0.1238
           leaveTangentWeight=0.2855
           tangentMode=RCTM_User
           time=0
           value=0.5
          }
          {arriveTangent=-1.5105, tangentMode=RCTM_User, time=0.85, value=2}
         ]}}}))
   )
   /SetVariables(
    Particles.Linear Color_Outline:LinearColor=&MakeLinearColorFromVectorAndFloat(
     "Vector (RGB)"=<0 0 0>
     "Float (Alpha)"=&FloatFromCurve(
      FloatCurve=DataInterface() {propertyValues {Curve {keys [
          {arriveTangentWeight=0.3333, interpMode=RCIM_Cubic, leaveTangentWeight=0.3333, tangentMode=RCTM_User, time=0, value=1}
          {tangentMode=RCTM_User, time=1, value=0}
         ]}}}
      Scale Curve=3
     )
    )
    Particles.SpriteSize_Outline:Vector2f=&Multiply_Vector2D(A=$Particles.SpriteSize, B=<1.2 1.02>)
   ) {isSetParameters=true}
   /SetVariables(
    Particles.Linear Color_Glow:LinearColor=&MultiplyLinearColor(
     Linear Color=$Particles.Color
     Scale Factor=&MakeVector4(
      X=1
      Y=1
      Z=1
      W=&FloatFromCurve(
       FloatCurve=DataInterface() {propertyValues {Curve {keys [
           {arriveTangent=-0.2204, interpMode=RCIM_Cubic, leaveTangent=-0.2204, leaveTangentWeight=0.2731, tangentMode=RCTM_User, time=0, value=1.75}
           {arriveTangent=-4.7379, tangentMode=RCTM_User, time=0.8, value=0.25}
          ]}}}
       Scale Curve=0.3
      )
     )
    )
    Particles.SpriteSize_Glow:Vector2f=&Multiply_Vector2D(A=$Particles.SpriteSize, B=<1.5 1.3>)
   ) {isSetParameters=true}
   /SetVariables(
    Particles.Linear Color_Glowy:LinearColor=&MultiplyLinearColor(
     Linear Color=$Particles.Color
     Scale Factor=&MakeVector4(
      X=0.5
      Y=0.5
      Z=0.5
      W=&FloatFromCurve(
       FloatCurve=DataInterface() {propertyValues {Curve {keys [
           {arriveTangent=-0.2204, interpMode=RCIM_Cubic, leaveTangent=-0.2204, leaveTangentWeight=0.2731, tangentMode=RCTM_User, time=0, value=1.75}
           {arriveTangent=-4.7379, tangentMode=RCTM_User, time=0.8, value=0.25}
          ]}}}
       Scale Curve=&RandomRangeFloat(Minimum=0.02, Maximum=0.05)
      )
     )
    )
    Particles.SpriteSize_Glowy:Vector2f=&Multiply_Vector2D(A=$Particles.SpriteSize, B=<1.1 1.15>)
    Particles.VisibilityTagLight:NiagaraInt32=&SpawnGroup()
   ) {isSetParameters=true}
   &FN_Light_Attributes(
    Color=&Multiply_LinearColorByFloat(Linear Color=$Particles.Color, Float=1)
    Create Color Attribute=true
    Create Exponent Attribute=true
    Create Position Attribute=true
    Create Radius Attribute=true
    Exponent=2
    Position=<0 -50 0>
    Radius=150
    Scale Color=&FN_Select_Float_from_Time_Of_Day(Morning=0.22, Day=0.25, Evening=0.22, Night=0.22)
   )
   &KillParticlesInVolume(Box Size:Vector3f=<50 50 100>, Invert Volume=true, Kill Shape=::NewEnumerator1, Kill Volume Enabled:NiagaraBool=$User.bIsFrontEndPreview)
  ]
  rendererClasses [&NiagaraSpriteRendererProperties, &NiagaraLightRendererProperties]
  renderers [
   {rendererClass=&NiagaraSpriteRendererProperties}
   {rendererClass=&NiagaraSpriteRendererProperties, rendererIndex=1}
   {rendererClass=&NiagaraSpriteRendererProperties, rendererIndex=2}
   {rendererClass=&NiagaraSpriteRendererProperties, rendererIndex=3}
   {rendererClass=&NiagaraLightRendererProperties, rendererIndex=4}
  ]
 }
}
summary {
 systemName=NS_Emote_PocketBobble_StraightSpeedLines
 emitters [
  {
   emitterName=Speed_Lines_Straight001
   rendererClasses [&NiagaraSpriteRendererProperties, &NiagaraLightRendererProperties]
  }
  {emitterName=FrontEnd_Ground_Glow, rendererClasses [&NiagaraSpriteRendererProperties]}
 ]
 userVariables [
  User.bIsFrontEnd {type=NiagaraBool}
  User.bIsFrontEndPreview {type=NiagaraBool}
  User.BurstIndex {type=NiagaraFloat, defaultValue=1, description=""}
  User.ColorVariant {type=NiagaraInt32}
  User.DelayTime {type=NiagaraFloat, defaultValue=0, description=""}
  User.FlashIntensity {type=NiagaraFloat, defaultValue=0.7, description=""}
  User.IsLocalPlayer {type=NiagaraBool, defaultValue=true}
  User.Pos_Offset_X {type=NiagaraFloat, defaultValue=0, description=""}
  User.Pos_Offset_Y {type=NiagaraFloat, defaultValue=0, description=""}
  User.RibbonIDOffset {type=NiagaraInt32}
  User.SizeMult {type=NiagaraFloat, defaultValue=1, description=""}
  User.SplashEroder {type=NiagaraFloat, defaultValue=0.7, description=""}
  User.Trail_Life {type=NiagaraFloat, defaultValue=0, description=""}
  User.WorkoutSelect {type=NiagaraFloat, defaultValue=0, description=""}
 ]
}
